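# Conversation database for Season 33 Winterfest Case 4, part 3, where the player questions an NPC about illegal yeast supplements.
# The first speech node requires an active Quest_S33_Winterfest_P4_Q03 and an NPC carrying the BatterBoi entry tag.
# A linear chain of three speech nodes, each with one player choice, leads to a Continue node.
# The final choice carries the quest's entry tag as a choice tag.
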ConversationDatabase /BattlePassS33_Quests/Conversations/Winterfest/Case_4/Quest_S33_Winterfest_P4_Q03
entry_points [{tag=Conversation.Entry.Season33.WF.Quest.04.part.03, destinations [#1]}]
nodes [
 ConversationEntryPointNode_1 {
  id=#1
  class=ConversationEntryPointNode
  outputs [#2]
  properties {entryTag {tagName=Conversation.Entry.Season33.WF.Quest.04.part.03}}
 }
 Speech_C_1 {
  id=#2
  class=Speech_C
  outputs [#3]
  properties {softTaskIcon=&T_UI_NPC_Intel2, generalConfig {defaultMessage="Hullo, friend! What seems to be the issue?"}}
  sub_nodes [
   ConversationChoiceNode_18 {class=ConversationChoiceNode, properties {defaultChoiceDisplayText=Wintervestigation}}
   FortConversationRequirement_ControllerRequirement_8 {
    class=FortConversationRequirement_ControllerRequirement
    properties {
     nodeName=Require: NPC Has Conversation Tag
     participantId {tagName=Conversation.Participant.NPC}
     requirement {class=FortControllerRequirement_NPCHasConversationEntryTag, properties {queryTag {tagName=Conversation.Entry.BatterBoi}}}
    }
   }
   FortConversationRequirement_ControllerRequirement_9 {
    class=FortConversationRequirement_ControllerRequirement
    properties {
     nodeName=Require: Is Quest Active
     requirement {
      class=FortControllerRequirement_IsQuestActive
      properties {questDefinition=/BattlePassS33_Quests/Items/QuestItems/Winterfest/Case_4/Quest_S33_Winterfest_P4_Q03}
     }
    }
   }
  ]
 }
 Speech_C_2 {
  id=#3
  class=Speech_C
  outputs [#4]
  properties {softTaskIcon=&T_UI_NPC_Intel2, generalConfig {defaultMessage="Nuh-uh, bud. These buns are naturally enriched. "}}
  sub_nodes [ConversationChoiceNode_0 {
    class=ConversationChoiceNode
    properties {defaultChoiceDisplayText=Are you taking illegal yeast supplements to give yourself unfair volume??}
   }]
 }
 Speech_C_6 {
  id=#4
  class=Speech_C
  outputs [#5]
  properties {
   softTaskIcon=&T_UI_NPC_Intel2
   generalConfig {defaultMessage="Yup! And unfair to who? I don't compete or anythin'. I just fight demons!"}
  }
  sub_nodes [ConversationChoiceNode_19 {class=ConversationChoiceNode, properties {defaultChoiceDisplayText=Really?}}]
 }
 Continue_C_2 {
  id=#5
  class=Continue_C
  properties {nodeName="", softTaskIcon=&T_UI_NPC_Intel2}
  sub_nodes [ConversationChoiceNode_20 {
    class=ConversationChoiceNode
    properties {defaultChoiceDisplayText=Huh. You've got me there., choiceTags {gameplayTags [{tagName=Conversation.Entry.Season33.WF.Quest.04.part.03}]}}
   }]
 }
]
